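; Gameplay ability for a Battle Pass S22 quest that credits players launched airborne by a directional launcher while holding a chicken.
; On activation from an event, it gets the owning player pawn's held object and casts it to the prey chicken NPC (Irwin Prey Nug).
; It then sends the WithChicken airborne directional-launch stat tag to the player's quest manager and ends the ability.
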(event Ability|EventActivateAbilityFromEvent
  (params (EventData :type Gameplay%20Event%20Data%20Structure%20%28by%20ref%29))
  (|CallParentFunction EventData)
  (bind _returnvalue (Fortnite|HeldObjects|GetHeldObjectAttachedtoPlayer (Variables|GAVars|Data|GetOwningPlayerPawn)))
  (Utilities|Casting|CastToNPC_Pawn_Irwin_Prey_Nug _returnvalue)
  (Quest|SendComplexCustomStatEvent (Variables|GAVars|Data|GetOwningPlayerQuestManager) 0 "(GameplayTags=((TagName=\"Athena.Quests.DirectionalLaunch.Airborne.WithChicken\")))" "(GameplayTags=)" 1)
  (Ability|EndAbility))
